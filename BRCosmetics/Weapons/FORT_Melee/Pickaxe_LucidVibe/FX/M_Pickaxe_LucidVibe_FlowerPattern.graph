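# Unlit translucent Niagara sprite material that draws a flower texture for the LucidVibe pickaxe effects.
# Particle color tints the texture's red channel, which feeds an emissive adjustment function.
# Opacity erodes that channel by inverse particle alpha, then scales it by a dynamic Opacity parameter.
# A shading path switch applies depth fade by default and skips it on mobile.
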Material /BRCosmetics/Weapons/FORT_Melee/Pickaxe_LucidVibe/FX/M_Pickaxe_LucidVibe_FlowerPattern
graph {
 properties {
  bAutomaticallySetUsageInEditor=false
  bUsedWithNiagaraSprites=true
  blendMode=BLEND_Translucent
  shadingModel=MSM_Unlit
  usedShadingModels=Unlit
 }
 nodes [
  n1 = ParticleColor()
  n2 = TextureSample() {texture=&/T_Pickaxe_LucidVibe_Flower_01}
  MaterialFunctionCall(Emissive=Multiply(A=@n1.RGB, B=@n2.R)) {materialFunction=&MF_FX_EmissiveAdjustments}
  ScalarParameter() {defaultValue=0.7, parameterName=Opacity}
  n3 = Multiply(A=Subtract(A=@n2.R, B=OneMinus(None=@n1.A)), B=DynamicParameter() {paramNames [Opacity,Param2,Param3,Param4]}.Opacity)
  ShadingPathSwitch(Default=DepthFade(Opacity=@n3) {fadeDistanceDefault=20}, Mobile=@n3)
 ]
}
