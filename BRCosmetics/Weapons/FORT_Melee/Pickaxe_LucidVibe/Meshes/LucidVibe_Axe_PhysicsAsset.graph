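# Physics asset for the LucidVibe axe mesh, a cosmetic two-handed melee pickaxe, giving it a simplified collision shape.
# A single body on the Melee_TwoHanded_Handle bone holds one capsule, offset mainly along the bone's Z axis and rotated.
# There are no constraints or other bodies, so the asset defines no jointed or secondary-motion parts.
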PhysicsAsset /BRCosmetics/Weapons/FORT_Melee/Pickaxe_LucidVibe/Meshes/LucidVibe_Axe_PhysicsAsset
bodies {Melee_TwoHanded_Handle [{shapeType=Capsule, length=75.3955, radius=23.4239, center=<0.7879 -0.9581 28.4602>, rotation=<rot -0.0027 -90 -0.0072>}]}
constraints []
